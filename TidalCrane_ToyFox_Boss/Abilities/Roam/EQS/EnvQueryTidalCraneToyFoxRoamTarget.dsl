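; EQS context that supplies roam targets for the TidalCrane ToyFox boss's Roam ability.
; ProvideActorsSet fetches actors tagged as roam targets from the global actor registry, avoiding a world search.
; It casts each to B_TidalCraneToyFoxRoamTarget and checks IsValidRoamTarget.
; As shown, the first valid target triggers a return of the local RetActors array, with no visible step adding actors to it.
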(fn ProvideActorsSet
  (params (QuerierObject :type Object%20Reference) (QuerierActor :type Actor%20Object%20Reference))
  (bind _outactors (Class|FortGlobalActorRegistrySubsystem|GetRegisteredActors (WorldSubsystems|GetFortGlobalActorRegistrySubsystem) (GameplayTags|MakeLiteralGameplayTag "(TagName=\"TidalCrane.ToyFox.Boss.Ability.Roam.Target\")")))
  (for _array_element _outactors
    (bind _asb_tidal_crane_toy_fox_roam_target (Utilities|Casting|CastToB_TidalCraneToyFoxRoamTarget _array_element))
    (bind _bvalid (Default|IsValidRoamTarget _asb_tidal_crane_toy_fox_roam_target))
    (if _bvalid
      (return (LocalVariables|GetRetActors)))))
